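# Translucent UI material that draws a circular context-state badge, a ring with inner fill and icon, for the frontend elevator UI.
# Two exponential radial gradients, with resolution-corrected softness, form the outer ring and adjustable-alpha fill.
# The icon texture supports rotation, centered scale and translation, defaulting to a checkmark.
# Its alpha joins the shape's opacity, and bIsIconColor switches it between ring color and its own tint.
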Material /Game/Athena/UI/Frontend/Spatial/Art/Elevator/M_UI_Elevator_ContextState_Icon
graph {
 properties {
  bUsedWithStaticMesh=false
  blendMode=BLEND_Translucent
  materialDomain=MD_UI
  pixelDepthOffsetMode=PDOM_Legacy
  usedShadingModels=Default Lit
 }
 nodes [
  Add()
  n1 = MaterialFunctionCall!(Rotation_Fraction=ScalarParameter() {group=Icon, parameterName=IconRotation}, TexCoord=TextureCoordinate()) {materialFunction=/FortUILibrary/Foundation/Art_ToolKit/AngledBoxes/MF_UI_Rotator_Fractions}
  n2 = ScalarParameter() {defaultValue=0.5, group=Icon, parameterName=ScaleIcon}
  n3 = ScalarParameter() {group=Icon, parameterName=Translate_Icon_X}
  n4 = ScalarParameter() {group=Icon, parameterName=Translate_Icon_Y}
  n5 = TextureSampleParameter2D(UVs=Add(
    A=MaterialFunctionCall(UVs=@n1.Result, Texture Scale=@n2) {materialFunction=&ScaleUVsByCenter}.UVs
    B=AppendVector(A=@n3, B=@n4)
   )) {group=Icon, parameterName=ContextState_Icon, texture=/Game/UI/Foundation/Textures/Icons/Checks/T-Icon-Check-32}
  n6 = TextureCoordinate()
  n7 = AppendVector(
   A=ScalarParameter() {defaultValue=0.5, group=CirlceTreatment, parameterName=CenterPosition_U}
   B=ScalarParameter() {defaultValue=0.5, group=CirlceTreatment, parameterName=CenterPosition_V}
  )
  n8 = Multiply(
   A=ScalarParameter() {defaultValue=18, group=CirlceTreatment, parameterName=Softness}
   B=MaterialFunctionCall!(Native_U_PixelSize=ScalarParameter() {defaultValue=40, group=CirlceTreatment, parameterName=NativePixelSize}) {materialFunction=/FortUILibrary/Foundation/Art_ToolKit/Tools/M_UI_Resolution_Sharpness_Correction}.Result
  )
  n9 = MaterialFunctionCall(UVs=@n6, CenterPosition=@n7, Radius=ScalarParameter() {defaultValue=0.4, group=InnerFill, parameterName=Radius_InnerFill}, Density=@n8) {materialFunction=&RadialGradientExponential}
  n10 = Saturate(None=OneMinus(None=Saturate(None=@n9.RadialGradientExponential)))
  n11 = Add(A=@n10, B=@n5.A)
  n12 = ScalarParameter() {group=Icon, parameterName=bIsIconColor}
  n13 = LinearInterpolate(
   A=VectorParameter() {group=InnerFill, parameterName=InnerFill_Color, defaultValue=<rgba 0 0.1667 1 1>}.RGB
   B=VectorParameter() {group=OuterRing, parameterName=OuterRing_Color, defaultValue=<rgba 0 0.6667 1 1>}.RGB
   Alpha=Saturate(None=LinearInterpolate(A=@n11, B=@n10, Alpha=@n12))
  )
  LinearInterpolate(
   A=@n13
   B=LinearInterpolate(A=@n13, B=VectorParameter() {group=Icon, parameterName=Icon_Color, defaultValue=<rgba 1 0.0644 0.2446 1>}.RGB, Alpha=@n5.A)
   Alpha=ScalarParameter() {group=Icon, parameterName=bIsIconColor}
  )
  n14 = Saturate(None=MaterialFunctionCall(UVs=@n6, CenterPosition=@n7, Radius=ScalarParameter() {defaultValue=0.5, group=OuterRing, parameterName=Radius_OuterRing}, Density=@n8) {materialFunction=&RadialGradientExponential}.RadialGradientExponential)
  n15 = ScalarParameter() {defaultValue=0.5, group=InnerFill, parameterName=InnerFill_Alpha}
  n16 = Add(A=DotProduct(A=@n10, B=@n14), B=Saturate(None=Multiply(A=@n14, B=@n15)))
  n17 = Saturate(None=Add(A=Saturate(None=@n16), B=@n5.A))
  n18 = ScalarParameter() {defaultValue=1, parameterName=Overall_Alpha}
  Reroute(None=Saturate(None=Multiply(A=@n17, B=@n18)))
 ]
}
